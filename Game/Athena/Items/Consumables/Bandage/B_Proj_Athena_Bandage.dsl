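; Projectile Blueprint for the Athena Bandage consumable that triggers the inherited delivery explosion when it lands.
; The stop and bounce events each call DoExplosion through their own DoOnce gate, so each path fires once but both can fire on the same projectile.
; BeginPlay, Hit, overlap, Tick and the construction script contain no logic.
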(event EventOnStop
  (params (Hit :type Hit%20Result%20Structure%20%28by%20ref%29))
  (Utilities|FlowControl|DoOnce false
    (:Completed
      (Delivery|DoExplosion))))

(event EventBeginPlay)

(event EventOnBounce
  (params (Hit :type Hit%20Result%20Structure%20%28by%20ref%29))
  (Utilities|FlowControl|DoOnce false
    (:Completed
      (Delivery|DoExplosion))))

(event Collision|EventHit
  (params (MyComp :type Primitive%20Component%20Object%20Reference) (Other :type Actor%20Object%20Reference) (OtherComp :type Primitive%20Component%20Object%20Reference) (bSelfMoved :type Boolean) (HitLocation :type Vector :default "0, 0, 0") (HitNormal :type Vector :default "0, 0, 0") (NormalImpulse :type Vector :default "0, 0, 0") (Hit :type Hit%20Result%20Structure%20%28by%20ref%29)))

(event Collision|EventActorBeginOverlap
  (params (OtherActor :type Actor%20Object%20Reference)))

(event EventTick
  (params (DeltaSeconds :type Float%20%28single-precision%29)))

(fn ConstructionScript)
